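# Post-process AnimBP for the StoryStern biped cosmetic companion, adding corrective poses and secondary motion to the incoming pose.
# A Pose Blend node applies the character's PoseAsset from the input pose, then chained AnimDynamics nodes simulate angular-spring ear and collar bones under reduced gravity.
# Constraints then position each deform thigh bone by equally weighted thigh and foot translations.
# The event graph is empty, and the DynamicsAlpha getter is unwired.
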AnimBlueprint /CosmeticCompanions/Assets/Biped/StoryStern/Meshes/M_MED_StoryStern_Biped_Post_AnimBP
target_skeleton=&/M_MED_StoryStern_Biped_Skeleton
graphs [
 AnimGraph {nodes [
   n1 = Input Pose()
   n2 = Get PoseAsset() -> (PoseAsset:Pose Asset Object Reference)
   n3 = Anim Dynamics - Bone: ear_02_l(
    ComponentPose=Local To Component(LocalPose=Pose!(PoseAsset:Pose Asset Object Reference=@n2, SourcePose:Pose Link Structure=@n1) -> (Pose:Pose Link Structure) {is=title AnimGraphNode_PoseBlendNode, bIgnorePoseAssetSourceAnimationCurves=true, blendOption=Linear})
    Alpha=1
    ExternalForce="0,0,0"
   ) {
    angularDampingOverride=1
    angularSpringConstant=20000
    bAngularSpring=true
    bOverrideAngularDamping=true
    bUseGravityOverride=true
    gravityScale=0.35
    simulationSpace=RootRelative
    boundBone {boneName=ear_02_l}
    componentLinearAccScale=<1 1 1>
    componentLinearVelScale=<1 1 1>
    gravityOverride {z=-980}
    physicsBodyDefinitions [{
      boundBone {boneName=ear_02_l}
      boxExtents=<20 30 30>
      constraintSetup {angularLimitsMax=<0 30 15>, angularLimitsMin=<0 -5 -15>, angularTarget=<1 0 0>}
      localJointOffset=<4 4 0>
     }]
   }
   n4 = Anim Dynamics - Bone: ear_03_l(ComponentPose=@n3, Alpha=1, ExternalForce="0,0,0") {
    angularDampingOverride=1
    angularSpringConstant=20000
    bAngularSpring=true
    bOverrideAngularDamping=true
    bUseGravityOverride=true
    gravityScale=0.35
    simulationSpace=RootRelative
    boundBone {boneName=ear_03_l}
    componentLinearAccScale=<1 1 1>
    componentLinearVelScale=<1 1 1>
    gravityOverride {z=-980}
    physicsBodyDefinitions [{
      boundBone {boneName=ear_03_l}
      boxExtents=<20 30 30>
      constraintSetup {angularLimitsMax=<0 20 15>, angularLimitsMin=<0 -5 -15>, angularTarget=<1 0 0>}
      localJointOffset=<4 4 0>
     }]
   }
   n5 = Anim Dynamics - Bone: ear_02_r(ComponentPose=@n4, Alpha=1, ExternalForce="0,0,0") {
    angularDampingOverride=1
    angularSpringConstant=20000
    bAngularSpring=true
    bOverrideAngularDamping=true
    bUseGravityOverride=true
    gravityScale=0.35
    simulationSpace=RootRelative
    boundBone {boneName=ear_02_r}
    componentLinearAccScale=<1 1 1>
    componentLinearVelScale=<1 1 1>
    gravityOverride {z=-980}
    physicsBodyDefinitions [{
      boundBone {boneName=ear_02_r}
      boxExtents=<20 30 30>
      constraintSetup {angularLimitsMax=<0 30 15>, angularLimitsMin=<0 -5 -15>, angularTarget=<1 0 0>}
      localJointOffset=<-4 -4 0>
     }]
   }
   n6 = Anim Dynamics - Bone: ear_03_r(ComponentPose=@n5, Alpha=1, ExternalForce="0,0,0") {
    angularDampingOverride=1
    angularSpringConstant=20000
    bAngularSpring=true
    bOverrideAngularDamping=true
    bUseGravityOverride=true
    gravityScale=0.35
    simulationSpace=RootRelative
    boundBone {boneName=ear_03_r}
    componentLinearAccScale=<1 1 1>
    componentLinearVelScale=<1 1 1>
    gravityOverride {z=-980}
    physicsBodyDefinitions [{
      boundBone {boneName=ear_03_r}
      boxExtents=<20 30 30>
      constraintSetup {angularLimitsMax=<0 20 15>, angularLimitsMin=<0 -5 -15>, angularTarget=<1 0 0>}
      localJointOffset=<-4 -4 0>
     }]
   }
   n7 = Anim Dynamics - Bone: deform_collar(ComponentPose=@n6, Alpha=1, ExternalForce="0,0,0") {
    angularDampingOverride=1
    angularSpringConstant=20000
    bAngularSpring=true
    bOverrideAngularDamping=true
    bUseGravityOverride=true
    gravityScale=0.35
    simulationSpace=RootRelative
    boundBone {boneName=deform_collar}
    componentLinearAccScale=<1 1 1>
    componentLinearVelScale=<1 1 1>
    gravityOverride {z=-980}
    physicsBodyDefinitions [{
      boundBone {boneName=deform_collar}
      boxExtents=<20 30 30>
      constraintSetup {angularLimitsMax=<0 10 2>, angularLimitsMin=<0 -5 -15>, angularTarget=<1 0 0>}
      localJointOffset=<4 0 0>
     }]
   }
   n8 = "Constraint - deform_thigh_l (TT)"!(ComponentPose=@n7, Alpha=1, ConstraintWeights_0=0.5, ConstraintWeights_1:float=0.5) -> (Pose) {
    boneToModify {boneName=deform_thigh_l}
    constraintSetup [{transformType=Translation, targetBone {boneName=thigh_l}}, {transformType=Translation, targetBone {boneName=foot_l}}]
    constraintWeights [1,1]
   }
   n9 = "Constraint - deform_thigh_r (TT)"!(ComponentPose=@n8, Alpha=1, ConstraintWeights_0=0.5, ConstraintWeights_1:float=0.5) -> (Pose) {
    boneToModify {boneName=deform_thigh_r}
    constraintSetup [{transformType=Translation, targetBone {boneName=thigh_r}}, {transformType=Translation, targetBone {boneName=foot_r}}]
    constraintWeights [1,1]
   }
   Output Pose(Result=Component To Local(ComponentPose=@n9))
   Get DynamicsAlpha() -> (DynamicsAlpha:float)
  ]}
 EventGraph {links [], nodes []}
]
variables [PoseAsset {type=PoseAsset, default=&/M_MED_StoryStern_Biped_PoseAsset}]
